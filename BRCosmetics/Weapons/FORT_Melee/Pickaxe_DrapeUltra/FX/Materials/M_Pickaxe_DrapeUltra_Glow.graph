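# Translucent, unlit, two-sided glow material for Niagara sprite and mesh particles on the DrapeUltra pickaxe.
# Particle color feeds an emissive-adjustment function, while particle alpha multiplies a channel-selected glow texture.
# That alpha is faded by a view-to-normal SmoothStep and DepthFade to soften grazing angles and intersections.
# A narrow SmoothStep on UV V, offset by the dynamic Mask parameter plus DebugMask, adds a controllable wipe along the surface.
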Material /BRCosmetics/Weapons/FORT_Melee/Pickaxe_DrapeUltra/FX/Materials/M_Pickaxe_DrapeUltra_Glow
graph {
 properties {
  bUsedWithNiagaraMeshParticles=true
  bUsedWithNiagaraSprites=true
  blendMode=BLEND_Translucent
  pixelDepthOffsetMode=PDOM_Legacy
  shadingModel=MSM_Unlit
  twoSided=true
  usedShadingModels=Unlit
 }
 nodes [
  n1 = ParticleColor()
  MaterialFunctionCall(Emissive=@n1.RGB) {materialFunction=&MF_FX_EmissiveAdjustments}
  n2 = Multiply(
   A=@n1.A
   B=ChannelMaskParameter(None=TextureSampleParameter2D() {parameterName=Glow Texture, texture=/BRCosmetics/Weapons/FORT_Melee/Pickaxe_DrapeUltra/FX/Textures/T_Pickaxe_DrapeUltra_Glow}.RGBA) {parameterName=TXParameter}.R
  )
  n3 = CameraVectorWS()
  n4 = VertexNormalWS()
  n5 = SmoothStep(
   Min=ScalarParameter() {parameterName=View Angle Fade Min, sliderMax=1}
   Max=ScalarParameter() {defaultValue=1, parameterName=View Angle Fade Max, sliderMax=1}
   Value=Abs(None=DotProduct(A=@n3, B=@n4))
  )
  n6 = Add(
   A=DynamicParameter() {defaultValue=<rgba 0 1 1 1>, paramNames [Mask,Param2,Param3,Param4]}.Mask
   B=ScalarParameter() {parameterName=DebugMask, sliderMax=1}
  )
  n7 = Add(A=ComponentMask(None=TextureCoordinate()) {g=true}, B=@n6)
  Multiply(
   A=DepthFade(Opacity=Multiply(A=@n2, B=@n5), FadeDistance=ScalarParameter() {parameterName=FadeDistance})
   B=Saturate(None=SmoothStep(Value=@n7) {constMax=0.1})
  )
  OneMinus(None=@n6)
 ]
}
